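# Translucent unlit material for a weak-point crack effect, usable on mesh particles.
# Color brightens along a vertical UV gradient, with a scrolling sine pulse scaled by PulseMultiplier and a static switch offering particle color for emissive.
# Vertex-offset math combines a ShrinkCrack-driven BlendMorphs function, shared hit-effect deformation, and rotation about a local axis based on radius-normalized distance from the object center.
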Material /BRCosmetics/Effects/Fort_Effects/BluePrints/WeakPointEffect/Crack
graph {
 properties {bUsedWithMeshParticles=true, blendMode=BLEND_Translucent, pixelDepthOffsetMode=PDOM_Legacy, shadingModel=MSM_Unlit, usedShadingModels=Unlit}
 nodes [
  n1 = OneMinus(None=ScalarParameter() {parameterName=ShrinkCrack})
  n2 = Add(A=Multiply(B=@n1), B=WorldPosition().XYZ)
  n3 = ObjectPositionWS()
  n4 = ObjectRadius()
  n5 = Multiply(A=Divide(A=Distance(A=@n2, B=@n3), B=Add(A=@n4, B=@n4))) {constB=0.3}
  Add(A=@n5, B=ScalarParameter() {parameterName=RandomRotation})
  n6 = VectorParameter() {parameterName=Color}
  n7 = MaterialFunctionCall() {materialFunction=&LinearGradient}
  n8 = Multiply(A=Add(A=@n6.RGB, B=Multiply(A=@n7.VGradient, B=@n7.VGradient)), B=@n7.VGradient)
  n9 = StaticBool()
  n10 = MaterialFunctionCall(
   Value=Add(
    A=@n7.VGradient
    B=MaterialFunctionCall() {materialFunction=&TimeWithSpeedVariable}.Frac Time
   )
   -1 to 1=@n9
   Sine Phase=@n9
  ) {materialFunction=&LinearSine}
  n11 = ScalarParameter() {parameterName=PulseMultiplier}
  Add(A=@n8, B=Multiply(A=@n8, B=Multiply(A=@n10."Rounded Linear Sine ", B=@n11)))
  n12 = MaterialFunctionCall!(In=@n1) {materialFunction=/Game/Effects/Fort_Effects/BluePrints/WeakPointEffect/BlendMorphs}
  n13 = MaterialFunctionCall!(
   WSDeformationCenter=VectorParameter() {group=z_code, parameterName=WSDeformationCenter}.RGB
   Code_DestructionAmount=ScalarParameter() {group=z_code, parameterName=Code_DestructionAmount}
   Code_UseUvForDamage
   Code_UseHitEffects
   Code_Emissive
   WSDamageRadius=ScalarParameter() {group=z_code, parameterName=WSDamageRadius}
   CrackDamageShader
   Diffuse
   ScalarModDamageCracks
   ScalarOffsetforDamageCracks
   ExposeLerpFunctionForInvertingTheHitDistanceCalc
   InvertDamageFalloff
   CrackWorldPosition
   WorldPosition
   DistanceCullFadeAnimationOffset
  ) {materialFunction=/Game/Packages/jl_tests/Fortnite_HitEffects}
  Add(A=@n12.Result, B=@n13.Hit And Shrink WPO)
  n14 = WorldPosition()
  Add(A=@n12.Result, B=@n14.XYZ)
  n15 = SphereMask(A=ComponentMask(None=@n7.VGradient) {g=true}, B=Constant() {r=0.2}) {attenuationRadius=0.19}
  Clamp(None=Add(A=@n15, B=Constant() {r=1}))
  Constant()
  Constant() {r=200}
  Constant()
  MaterialFunctionCall(Emissive=StaticSwitchParameter(True=ParticleColor().RGB, False=@n6.RGB) {parameterName=Particle}) {materialFunction=&MF_FX_EmissiveAdjustments}
  Multiply(A=@n6.RGB) {constB=5}
  Multiply(A=Add(A=Multiply(A=@n7.VGradient, B=@n6.RGB), B=@n6.RGB), B=OneMinus(None=@n15))
  OneMinus(None=@n7.VGradient)
  n16 = Constant2Vector()
  n17 = Clamp(None=ComponentMask(None=Subtract(A=@n3, B=@n2)) {b=true})
  n18 = Multiply(A=ConstantBiasScale(None=Ceil(None=@n17)) {bias=-0.5, scale=2}, B=ObjectRadius())
  RotateAboutAxis(
   NormalizedRotationAxis=Transform(None=Constant3Vector() {constant=<rgba 0 1 0 0>}) {transformSourceType=TRANSFORMSOURCE_Local}
   RotationAngle=@n5
   PivotPoint=Subtract(A=ObjectPositionWS(), B=AppendVector(A=@n16, B=@n18))
   Position=WorldPosition().XYZ
  )
  ScalarParameter() {parameterName=EnlargeCrack}
  ScalarParameter() {group=z_code, parameterName=Code_HitAnimation}
  StaticBoolParameter() {defaultValue=true, group=z_code, parameterName=Code_UseHitEffects}
  StaticSwitchParameter(True=@n12.Result, False=Constant()) {defaultValue=true, parameterName=UseVertexShader}
  Subtract(A=@n13.Hit And Shrink WPO, B=@n14.XYZ)
  VectorParameter() {group=z_code, parameterName=Code_DeformationVector}
 ]
}
